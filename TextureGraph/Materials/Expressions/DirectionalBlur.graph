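# Unlit TextureGraph expression material that applies a directional blur to an input texture.
# The Input texture parameter and default texture coordinates feed the MF_DirectionalBlur material function, with the Angle parameter setting blur direction.
# The Strength parameter is scaled by a constant to give the blur amount, while sigma and step count are hardcoded rather than exposed.
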Material /TextureGraph/Materials/Expressions/DirectionalBlur
graph {
 properties {pixelDepthOffsetMode=PDOM_Legacy, shadingModel=MSM_Unlit, usedShadingModels=Unlit}
 nodes [MaterialFunctionCall!(
   InputTexture=TextureObjectParameter() {parameterName=Input, samplerType=SAMPLERTYPE_LinearColor, texture=/TextureGraph/Textures/default/default_Roughness}
   Coordinates=TextureCoordinate()
   BlurAngle=ScalarParameter() {parameterName=Angle, sliderMax=180, sliderMin=-180}
   BlurAmount=Multiply(A=ScalarParameter() {defaultValue=0.1, parameterName=Strength, sliderMax=1}, B=Constant() {r=20})
   Sigma=Constant() {r=2}
   BlurSteps=Constant() {r=50}
  ) {materialFunction=/TextureGraph/Materials/Util/MF_DirectionalBlur}]
}
