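# Material function that returns a Gaussian-blurred RGB sample of an input texture at given coordinates, with an adjustable radius.
# A Custom HLSL node reads the texture dimensions and loops over a square texel neighborhood in one non-separable pass, so samples grow quadratically with radius.
# It normalizes the weighted sum using a hardcoded sigma of 30.
# The wired Sigma input is ignored and the TexelSize node is unconnected.
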MaterialFunction /TextureGraph/Materials/Util/MF_GaussianBlur
graph {nodes [
  MaterialExpressionFunctionInput_0 = FunctionInput(Preview=TextureObject() {samplerType=SAMPLERTYPE_LinearColor, texture=/TextureGraph/Textures/default/default_Roughness}) {bUsePreviewValueAsDefault=true, inputName=InputTexture, inputType=FunctionInput_Texture2D, sortPriority=-1}
  MaterialExpressionFunctionInput_1 = FunctionInput(Preview=TextureCoordinate()) {bUsePreviewValueAsDefault=true, inputName=Coordinates, inputType=FunctionInput_Vector2, sortPriority=1}
  MaterialExpressionConstant_4 = Constant() {r=6}
  MaterialExpressionFunctionInput_6 = FunctionInput(Preview=Constant() {r=6}) {bUsePreviewValueAsDefault=true, inputName=Radius, inputType=FunctionInput_Scalar, sortPriority=3}
  n1 = Custom!(
   Tex=@MaterialExpressionFunctionInput_0
   UV=@MaterialExpressionFunctionInput_1
   Sigma=@MaterialExpressionConstant_4
   Radius=@MaterialExpressionFunctionInput_6
  ) {
   code="//Gaussian Blur\r\n\r\nstruct GaussianBlur\r\n{\r\n  //Algorithm from https://www.pixelstech.net/article/1353768112-Gaussian-Blur-Algorithm\r\nfloat NormalDistribution(float x, float y)\r\n{\r\n\tconst float sigma = 30;\r\n\r\n\tfloat sigmaSquare = pow(sigma, 2);\r\n\r\n\tconst float expConstant = 2.71828;\r\n\tconst float PI = 3.141592653;\r\n\t\r\n\tfloat exponent = (-(pow(x, 2) + pow(y, 2))) / (2 * sigmaSquare);\r\n\r\n\tfloat gau = pow(expConstant, exponent) / (2 * PI * sigmaSquare);\r\n\treturn gau;\r\n}\r\n\r\nfloat NormalDistribution1D(float x)\r\n{\r\n\treturn exp(-0.5 * pow(3.141 * (x), 2));\r\n}\r\n\r\n\r\nfloat4 Run(Texture2D SourceTexture,SamplerState TextureSampler, float2 texelUV, float2 texelPos, int radius, int width, int height)\r\n{\r\n    float output = 0;\r\n\tint diameter = radius * 2;\r\n\tfloat4 pixelSum = float4(0, 0, 0, 0);\r\n\tfloat weightSum = 0;\r\n\t\r\n\t\tfloat widthInv = rcp(float(width));\r\n\tfloat heightInv = rcp(float(height));\r\n\t\r\n\tfloat2 centerSourceUV = texelUV;\r\n\t\r\n\t// (-r, -r)\t\t\t(r, -r)\r\n\t//\t\r\n\t//\t\t\t(0, 0)\r\n\t//\r\n\t// (-r, r)\t\t\t(r, r)\r\n\t// 1) This iteratres over every pixel from (-r, -r) to (r, r)\r\n\t// 2) and for each pixel, it finds the Weight through NormalDistrubution formula for that particular pixel location\r\n\t// 3) it then multiplies the Pixel Value of that location with the Weight of that location\r\n\t// 4) then at the end, the calculated gaussian value for pixel (0, 0) is the average of all pixel values by all their weights. \r\n\t// 5) Through this we find weighted average of a pixel with a particular kernel\r\n\r\n\tfor (int px = -radius; px <= radius; px++)\r\n\t{\r\n\t\tfor (int py = -radius; py <= radius; py++)\r\n\t\t{\r\n\t\t\tint2 neighborTexel = int2(px + texelPos.x, py + texelPos.y); //Given the radius, find the texel for which it needs to calculate weight\r\n\r\n\t\t\tfloat2 NewUV = centerSourceUV + float2(px * widthInv, py * heightInv);\r\n\t\t\t\r\n\t\t\t//float4 pixelValue = SourceTexture.Sample(SamplerStates_Linear_Clamp, uv);\r\n\r\n           //float4 pixelValue = SourceTexture.SampleGrad(GetMaterialSharedSampler(SamplerStates_Linear_Clamp,View.MaterialTextureBilinearClampedSampler), uv, 0, 0);\r\n           float4 pixelValue = Texture2DSample(SourceTexture, GetMaterialSharedSampler(TextureSampler,View.MaterialTextureBilinearClampedSampler), NewUV);\r\n           \r\n           float weight = NormalDistribution(px, py); // Calculate the weight for given location\r\n\t\t\tpixelSum += pixelValue * weight; // Aggregate pixel value\r\n\t\t\tweightSum += weight; // Aggregate weight\r\n\t\t}\r\n\t}\r\n\t\r\n\treturn pixelSum / weightSum;\r\n}};\r\n\r\n\r\nGaussianBlur GBlur;\r\n\r\nfloat2 Dimensions;\r\n\r\nTex.GetDimensions(Dimensions.x, Dimensions.y);\r\n\r\nfloat2 sourceTexelPos = UV * Dimensions; \r\n\t\r\nfloat4 MaskVal = GBlur.Run(Tex, TexSampler, UV, sourceTexelPos, Radius, Dimensions.x, Dimensions.y);\r\n\t\r\nreturn float3(MaskVal.r, MaskVal.g, MaskVal.b);\r\n\r\n//return float3(0.5,0.5,0.5);\r\n"
   description=Gaussian Blur
   showCode=true
   inputs [
    {inputName=Tex, input {expression=&self:MaterialExpressionFunctionInput_0}}
    {inputName=UV, input {expression=&self:MaterialExpressionFunctionInput_1}}
    {inputName=Sigma, input {expression=&self:MaterialExpressionConstant_4}}
    {inputName=Radius, input {expression=&self:MaterialExpressionFunctionInput_6}}
   ]
  }
  FunctionOutput(None=@n1) {outputName=BlurredTexture}
  TextureProperty(Tex=@MaterialExpressionFunctionInput_0) {property=TMTM_TexelSize}
 ]}
