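# GaussianBlur is an unlit TextureGraph expression material that applies a Gaussian blur to an input texture.
# An Input texture object parameter, the default texture coordinates and a Radius scalar parameter feed the MF_GaussianBlur material function.
# The blur work lives inside that function, so the material only exposes the parameters and outputs its result.
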Material /TextureGraph/Materials/Expressions/GaussianBlur
graph {
 properties {pixelDepthOffsetMode=PDOM_Legacy, shadingModel=MSM_Unlit, usedShadingModels=Unlit}
 nodes [MaterialFunctionCall!(
   InputTexture=TextureObjectParameter() {parameterName=Input, samplerType=SAMPLERTYPE_LinearColor, texture=/TextureGraph/Textures/default/default_Roughness}
   Coordinates=TextureCoordinate()
   Radius=ScalarParameter() {defaultValue=1, parameterName=Radius, sliderMax=100, sliderMin=1}
  ) {materialFunction=/TextureGraph/Materials/Util/MF_GaussianBlur}]
}
