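# Texture Graph utility material function that applies a radial (zoom) blur to an input texture and outputs BlurredTexture.
# A Custom HLSL node averages a fixed 50 samples stepped from each UV toward CenterX/CenterY, using the shared bilinear clamped sampler.
# BlurStrength scales the step span, weighted by distance from the center over Radius (saturated), so Radius is in UV units.
# TexelSize is wired into the node but unused by the code.
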MaterialFunction /TextureGraph/Materials/Util/MF_RadialBlur
graph {nodes [
  MaterialExpressionFunctionInput_0 = FunctionInput(Preview=TextureObject() {samplerType=SAMPLERTYPE_LinearColor, texture=/TextureGraph/Textures/default/default_Roughness}) {bUsePreviewValueAsDefault=true, inputName=InputTexture, inputType=FunctionInput_Texture2D, sortPriority=-1}
  MaterialExpressionTextureProperty_0 = TextureProperty(Tex=@MaterialExpressionFunctionInput_0) {property=TMTM_TexelSize}
  MaterialExpressionFunctionInput_1 = FunctionInput(Preview=TextureCoordinate()) {bUsePreviewValueAsDefault=true, inputName=Coordinates, inputType=FunctionInput_Vector2, sortPriority=1}
  MaterialExpressionFunctionInput_6 = FunctionInput(Preview=Constant() {r=6}) {bUsePreviewValueAsDefault=true, inputName=Radius, inputType=FunctionInput_Scalar, sortPriority=3}
  MaterialExpressionConstant_4 = Constant() {r=50}
  MaterialExpressionFunctionInput_2 = FunctionInput(Preview=Constant() {r=1}) {bUsePreviewValueAsDefault=true, inputName=BlurStrength, inputType=FunctionInput_Scalar, sortPriority=3}
  MaterialExpressionFunctionInput_8 = FunctionInput(Preview=Constant() {r=0.5}) {bUsePreviewValueAsDefault=true, inputName=CenterX, inputType=FunctionInput_Scalar, sortPriority=3}
  MaterialExpressionFunctionInput_9 = FunctionInput(Preview=Constant() {r=0.5}) {bUsePreviewValueAsDefault=true, inputName=CenterY, inputType=FunctionInput_Scalar, sortPriority=3}
  n1 = Custom!(
   Tex=@MaterialExpressionFunctionInput_0
   TexelSize=@MaterialExpressionTextureProperty_0
   UV=@MaterialExpressionFunctionInput_1
   Radius=@MaterialExpressionFunctionInput_6
   Samples=@MaterialExpressionConstant_4
   BlurStrength=@MaterialExpressionFunctionInput_2
   CenterX=@MaterialExpressionFunctionInput_8
   CenterY=@MaterialExpressionFunctionInput_9
  ) {
   code="//Radial Blur\r\n\r\n//return float4(UV, 0, 1);\r\n\r\nfloat4 col = float4(0,0,0,0);\r\nfloat2 dist = UV - float2(CenterX, CenterY);\r\n\r\nfor(int j = 0; j < Samples; j++) \r\n{\r\n     float scale = 1 - BlurStrength * (j / Samples)* (saturate(length(dist) / Radius));\r\n     //col += Tex.Sample(SamplerStates_Linear_Clamp, dist * (scale) + float2(CenterX, CenterY));\r\n     //col += Tex.SampleGrad(GetMaterialSharedSampler(SamplerStates_Linear_Clamp,View.MaterialTextureBilinearClampedSampler), dist * (scale) + float2(CenterX, CenterY), 0, 0);\r\n     col += Texture2DSample(Tex, GetMaterialSharedSampler(TexSampler,View.MaterialTextureBilinearClampedSampler), dist * (scale) + float2(CenterX, CenterY));\r\n}\r\n\r\ncol /= Samples;                \r\nreturn col;"
   description=Radial Blur
   showCode=true
   inputs [
    {inputName=Tex, input {expression=&self:MaterialExpressionFunctionInput_0}}
    {inputName=TexelSize, input {expression=&self:MaterialExpressionTextureProperty_0}}
    {inputName=UV, input {expression=&self:MaterialExpressionFunctionInput_1}}
    {inputName=Radius, input {expression=&self:MaterialExpressionFunctionInput_6}}
    {inputName=Samples, input {expression=&self:MaterialExpressionConstant_4}}
    {inputName=BlurStrength, input {expression=&self:MaterialExpressionFunctionInput_2}}
    {inputName=CenterX, input {expression=&self:MaterialExpressionFunctionInput_8}}
    {inputName=CenterY, input {expression=&self:MaterialExpressionFunctionInput_9}}
   ]
  }
  FunctionOutput(None=@n1) {outputName=BlurredTexture}
 ]}
