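# Unlit material that serves as the radial blur expression in TextureGraph, applying the effect to an input texture.
# It wraps the MF_RadialBlur material function, feeding it the Input texture parameter, sampled as linear color, along with default texture coordinates.
# Strength and Radius scalar parameters control the blur, while constants fix its center at the middle of UV space.
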Material /TextureGraph/Materials/Expressions/RadialBlur
graph {
 properties {pixelDepthOffsetMode=PDOM_Legacy, shadingModel=MSM_Unlit, usedShadingModels=Unlit}
 nodes [MaterialFunctionCall!(
   InputTexture=TextureObjectParameter() {parameterName=Input, samplerType=SAMPLERTYPE_LinearColor, texture=/TextureGraph/Textures/default/default_AO}
   Coordinates=TextureCoordinate()
   BlurStrength=ScalarParameter() {defaultValue=0.1, parameterName=Strength, sliderMax=1}
   CenterX=Constant() {r=0.5}
   CenterY=Constant() {r=0.5}
   Radius=ScalarParameter() {defaultValue=1, parameterName=Radius, sliderMax=1}
  ) {materialFunction=/TextureGraph/Materials/Util/MF_RadialBlur}]
}
